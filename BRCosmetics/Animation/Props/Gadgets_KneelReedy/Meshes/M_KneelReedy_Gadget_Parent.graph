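# Parent material for the KneelReedy gadget prop, adding a torus-style Fresnel highlight to the shared 2023 character material stack.
# Pre-skinned local position, interpolated and transformed, is crossed with Z to give a tangential axis for AxisAlignedFresnel, with parameters for falloff and strength.
# That result color-dodges the emissive output of the PreFX function, and the modified attributes feed PostFX.
# A quality switch node sits unconnected.
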Material /BRCosmetics/Animation/Props/Gadgets_KneelReedy/Meshes/M_KneelReedy_Gadget_Parent
graph {
 properties {
  bUseMaterialAttributes=true
  bUsedWithMeshDeformer=true
  bUsedWithMorphTargets=true
  bUsedWithSkeletalMesh=true
  pixelDepthOffsetMode=PDOM_Legacy
  previewMesh=&/KneelReedy_Gadget
  shadingModel=MSM_FromMaterialExpression
  usedShadingModels=Default Lit | Subsurface | Clear Coat | Subsurface Profile | Cloth
 }
 nodes [
  n1 = MaterialFunctionCall() {materialFunction=&MF_CharacterParent_2023_PreFX}
  n2 = GetMaterialAttributes(None=@n1.Result) {attributeGetTypes [#1]}
  n3 = Constant3Vector() {constant=<rgba 0 0 1 0>}
  n4 = Transform(None=VertexInterpolator(VS=LocalPosition() {localOrigin=InstancePreSkinning}.XYZ).PS) {transformSourceType=TRANSFORMSOURCE_World, transformType=TRANSFORM_Local}
  n5 = Multiply(
   A=MaterialFunctionCall(Axis=CrossProduct(A=@n3, B=@n4), Exponent=ScalarParameter() {defaultValue=2.4736, parameterName=Torus - Fresnel Factor}) {materialFunction=&AxisAlignedFresnel}.Fresnel
   B=ScalarParameter() {defaultValue=0.5, parameterName=Torus - Fresnel Strength}
  )
  MaterialFunctionCall(
   Material=SetMaterialAttributes(
    MaterialAttributes=@n2.MaterialAttributes
    Emissive Color=MaterialFunctionCall!(Blend=@n5, Base=@n2.Emissive Color) {materialFunction=/Engine/Functions/Engine_MaterialFunctions03/Blends/Blend_ColorDodge}.Result
   ) {attributeSetTypes [#1]}
   M=@n1.M
   Pre Skinned Local Position=@n1.M
   Distance Blend=@n1.M
   Fresnel=@n1.M
  ) {materialFunction=&MF_CharacterParent_2023_PostFX}
  MaterialFunctionCall!(Low _ PC Low and Med, "High _ XB1, PS4", "Epic _ PC High, XSX, PS5", --------, "Mobile Low _ PC Perf, Android", Mobile Med _ Switch) {materialFunction=/Game/Global/MaterialFunctions/TechArt/MF_Fortnite_QualitySwitch_Scalar}
 ]
}
